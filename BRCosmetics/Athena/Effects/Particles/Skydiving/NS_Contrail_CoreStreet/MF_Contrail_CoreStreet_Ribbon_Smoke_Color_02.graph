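# Material function that computes smoke color for the CoreStreet skydiving contrail ribbon.
# Sun/moon direction dotted with a world normal input drives two light/dark color sets, blended by an inverted Fresnel whose exponent comes from panning cloud noise.
# A powered UV gradient optionally lerps the result toward particle color.
# The EmissiveNoise input and ShadowMult multiply do not feed the output.
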MaterialFunction /BRCosmetics/Athena/Effects/Particles/Skydiving/NS_Contrail_CoreStreet/MF_Contrail_CoreStreet_Ribbon_Smoke_Color_02
graph {nodes [
  n1 = MaterialFunctionCall(Float2=TextureCoordinate()) {materialFunction=&BreakOutFloat2Components}
  n2 = Multiply(A=@n1.R, B=ScalarParameter() {defaultValue=1})
  n3 = Multiply(A=@n1.G, B=ScalarParameter() {defaultValue=1, parameterName=Param_1})
  n4 = Time()
  n5 = TextureSample(UVs=Panner(Coordinate=AppendVector(A=@n2, B=@n3), Time=Multiply(A=@n4)) {speedX=0.5}) {texture=/Game/Global/Volumetric/T_Cloud2dNoise}
  n6 = ScalarParameter() {defaultValue=1, parameterName=Param_2}
  n7 = MaterialFunctionCall(
   Brightness=ScalarParameter() {defaultValue=2, parameterName=ColorFresnelBrightnes}
   Exponent=Multiply(A=Power(Base=@n5.R, Exp=@n6) {constExponent=1}, B=ScalarParameter() {defaultValue=3, parameterName=Param_3}) {constB=3}
  ) {materialFunction=&MF_Fresnel_FX}
  Clamp(None=@n7.Fresnel Inverted)
  FunctionInput() {inputName=EmissiveNoise}
  n8 = CollectionParameter() {
   collection=&FortniteMaterialParameters
   parameterName=SunAndMoonModelDirectionalVector
  }
  n9 = Normalize(VectorInput=ComponentMask(None=@n8) {b=true, g=true, r=true})
  n10 = FunctionInput(Preview=Constant3Vector() {constant=<rgba 0 0 1 0>}) {bUsePreviewValueAsDefault=true, inputName=WorldSpaceNormal}
  n11 = Power(Base=Clamp(None=DotProduct(A=@n9, B=@n10)), Exp=ScalarParameter() {defaultValue=2, group=Color, parameterName=HighlightPower})
  n12 = Add(
   A=Multiply(A=@n11, B=VectorParameter() {parameterName=LightColor, defaultValue=<rgba 0 0 0 1>}.RGB) {constB=0}
   B=VectorParameter() {parameterName=DarkColor, defaultValue=<rgba 0 0 0 1>}.RGB
  )
  n13 = Multiply(A=@n11, B=VectorParameter() {parameterName=LightColor_02, defaultValue=<rgba 0 0 0 1>}.RGB) {constB=0}
  n14 = VectorParameter() {parameterName=DarkColor_02, defaultValue=<rgba 0 0 0 1>}
  n15 = LinearInterpolate(
   A=@n12
   B=LinearInterpolate(A=@n12, B=Add(A=@n13, B=@n14.RGB), Alpha=@n7.Fresnel Inverted)
   Alpha=ScalarParameter() {parameterName=UseFresnelColor}
  )
  n16 = ParticleColor()
  n17 = TextureCoordinate()
  n18 = AppendVector(A=ScalarParameter() {parameterName=LerpParticleColorGrad_UAdd}, B=ScalarParameter() {parameterName=LerpParticleColorGrad_VAdd})
  n19 = OneMinus(None=MaterialFunctionCall(UV Channel=Add(A=@n17, B=@n18)) {materialFunction=&LinearGradient}.UGradient)
  n20 = ScalarParameter() {defaultValue=5, parameterName=LerpParticleColorGrad_Power}
  n21 = Saturate(None=Multiply(A=Power(Base=@n19, Exp=@n20), B=ScalarParameter() {defaultValue=5, parameterName=LerpParticleColorGrad_Mullt}))
  FunctionOutput(None=LinearInterpolate(A=@n15, B=LinearInterpolate(A=@n15, B=@n16.RGB, Alpha=@n21), Alpha=ScalarParameter() {parameterName=LerpToParticleColor}))
  Multiply(B=ScalarParameter() {defaultValue=0.5, parameterName=ShadowMult})
 ]}
