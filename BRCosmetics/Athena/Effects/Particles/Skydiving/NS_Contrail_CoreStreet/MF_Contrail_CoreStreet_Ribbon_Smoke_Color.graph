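# Material function for the CoreStreet skydiving contrail that supplies color for the ribbon smoke effect.
# It is a thin wrapper that forwards an EmissiveNoise input and a Normal_2 input into MF_Contrail_CoreStreet_Ribbon_Smoke_Color_02 and outputs that function's Result.
# Normal_2 feeds the nested WorldSpaceNormal input and falls back to a +Z vector when unconnected.
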MaterialFunction /BRCosmetics/Athena/Effects/Particles/Skydiving/NS_Contrail_CoreStreet/MF_Contrail_CoreStreet_Ribbon_Smoke_Color
graph {nodes [
  n1 = Constant3Vector()
  n2 = Constant3Vector() {constant=<rgba 0 0 1 0>}
  FunctionOutput(None=MaterialFunctionCall!(
    EmissiveNoise=FunctionInput(Preview=@n1) {inputName=EmissiveNoise}
    WorldSpaceNormal=FunctionInput(Preview=@n2) {bUsePreviewValueAsDefault=true, inputName=Normal_2}
   ) {materialFunction=&/MF_Contrail_CoreStreet_Ribbon_Smoke_Color_02}.Result)
 ]}
