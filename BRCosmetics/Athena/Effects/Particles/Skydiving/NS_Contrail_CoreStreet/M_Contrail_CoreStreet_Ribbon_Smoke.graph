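# Translucent unlit Niagara ribbon material for smoke in the CoreStreet skydiving contrail.
# Panning storm-cloud samples with bump-offset parallax supply normal and glow masks; the transformed normal and contrast-adjusted noise feed a shared color function.
# Opacity adds a noise-distorted gradient ramp to alpha erosion, then multiplies by gradient-masked fresnel and particle alpha.
# Dynamic parameters supply pan speed, dive blend and UV offset, with UOffset driving both axes and VOffset unused.
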Material /BRCosmetics/Athena/Effects/Particles/Skydiving/NS_Contrail_CoreStreet/M_Contrail_CoreStreet_Ribbon_Smoke
graph {
 properties {bUsedWithNiagaraRibbons=true, blendMode=BLEND_Translucent, pixelDepthOffsetMode=PDOM_Legacy, shadingModel=MSM_Unlit, usedShadingModels=Unlit}
 nodes [
  n1 = VectorParameter() {parameterName=NormalMult, defaultValue=<rgba 1 1 1 0>}
  n2 = NamedRerouteUsage()
  MaterialFunctionCall!(
   EmissiveNoise=MaterialFunctionCall(In=NamedRerouteUsage(), Contrast=Constant() {r=0.5}) {materialFunction=&CheapContrast}.Result
   Normal_2=Transform(None=Multiply(A=@n1.RGB, B=@n2))
  ) {materialFunction=&/MF_Contrail_CoreStreet_Ribbon_Smoke_Color}
  n3 = Constant() {r=0.5}
  n4 = Constant()
  n5 = MaterialFunctionCall(UV Channel=Subtract(A=TextureCoordinate(), B=AppendVector(A=@n3, B=@n4))) {materialFunction=&LinearGradient}
  n6 = LinearInterpolate(
   A=MaterialFunctionCall(
    Brightness=ScalarParameter() {defaultValue=1, parameterName=fresnel_brightness}
    Exponent=ScalarParameter() {defaultValue=1, parameterName=fresnel_exponent}
   ) {materialFunction=&MF_Fresnel_FX}.Fresnel
   Alpha=Saturate(None=Multiply(A=@n5.UGradient))
  )
  n7 = ScalarParameter() {defaultValue=1, parameterName=Use Fresnel Opacity}
  n8 = TextureCoordinate()
  n9 = MaterialFunctionCall(Float2=TextureCoordinate()) {materialFunction=&BreakOutFloat2Components}
  n10 = AppendVector(
   A=Multiply(A=@n9.R, B=ScalarParameter() {defaultValue=1, parameterName=BottomRamp_Distortion_UTile})
   B=Multiply(A=@n9.G, B=ScalarParameter() {defaultValue=2, parameterName=BottomRamp_Distortion_VTile})
  )
  n11 = DynamicParameter() {paramNames [UOffset,VOffset,PanSpeed,DiveLerp]}
  n12 = AppendVector(A=@n11.UOffset, B=@n11.UOffset)
  n13 = TextureSample(UVs=Panner(Coordinate=Add(A=@n10, B=@n12), Time=NamedRerouteUsage()) {speedX=0.8}) {texture=/Game/Global/Volumetric/T_Cloud2dNoise}
  n14 = Multiply(A=AppendVector(A=@n13.R, B=@n13.R), B=ScalarParameter() {defaultValue=0.1, parameterName=BottomRamp_Distortion_Mult})
  n15 = Power(
   Base=MaterialFunctionCall(UV Channel=Add(A=@n8, B=@n14)) {materialFunction=&LinearGradient}.UGradient
   Exp=ScalarParameter() {defaultValue=1, parameterName=BottomRamp_Add_Power}
  ) {constExponent=10}
  n16 = ScalarParameter() {defaultValue=1, parameterName=BottomRamp_Add_Mult}
  n17 = LinearInterpolate(B=Saturate(None=Multiply(A=@n15, B=@n16)), Alpha=ScalarParameter() {defaultValue=1, parameterName=UseBottomRamp})
  n18 = MaterialFunctionCall(Float2=TextureCoordinate()) {materialFunction=&BreakOutFloat2Components}
  n19 = Multiply(
   A=Power(Base=OneMinus(None=@n18.R), Exp=ScalarParameter() {defaultValue=1, parameterName=ErosionRamp_Power}) {constExponent=1}
   B=ScalarParameter() {defaultValue=1, parameterName=ErosionRamp_Mult}
  )
  n20 = LinearInterpolate(
   B=MaterialFunctionCall!(
    "Erosion Texture (RGBA)"=@n13.RGBA
    Erosion Contrast=ScalarParameter() {defaultValue=0.5, parameterName=Erosion Contrast}
    Erosion Amount=LinearInterpolate(Alpha=@n19)
   ) {materialFunction=&/MF_Contrail_CoreStreet_AlphaErosion_Complex}.Color
   Alpha=ScalarParameter() {defaultValue=1, parameterName=UseErosion}
  ) {constA=1}
  Multiply(A=ParticleColor().A, B=Multiply(A=LinearInterpolate(B=@n6, Alpha=@n7) {constA=1}, B=Add(A=@n17, B=@n20)))
  NamedRerouteDeclaration(None=Panner(Coordinate=NamedRerouteUsage(), Time=NamedRerouteUsage()) {speedX=1}) {name=pan, nodeColor=<rgba 0 1 0.2031 1>}
  NamedRerouteDeclaration(None=Multiply(A=Time(), B=ScalarParameter() {defaultValue=1, parameterName=PanSpeed})) {name=time speed, nodeColor=<rgba 1 0 0.1172 1>}
  n21 = DynamicParameter() {paramNames [UOffset,VOffset,PanSpeed,DiveLerp]}
  NamedRerouteDeclaration(None=@n21.DiveLerp) {name=divelerp, nodeColor=<rgba 1 0 0.9844 1>}
  n22 = TextureCoordinate()
  n23 = AppendVector(A=ScalarParameter() {defaultValue=2, parameterName=TileX}, B=ScalarParameter() {defaultValue=1, parameterName=TileY})
  NamedRerouteDeclaration(None=Add(A=Multiply(A=@n22, B=@n23), B=AppendVector(A=@n21.UOffset, B=@n21.UOffset))) {name=coords, nodeColor=<rgba 0 0.4375 1 1>}
  n24 = LinearInterpolate(
   A=TextureSample(UVs=NamedRerouteUsage()) {samplerType=SAMPLERTYPE_LinearGrayscale, texture=/Game/VisualThreatManager/StormVisuals/LargeClouds/T_StormCloudDetail01_M}.G
   B=TextureSample(UVs=NamedRerouteUsage()) {samplerType=SAMPLERTYPE_LinearGrayscale, texture=&T_StormCloudDetail01_M}.G
   Alpha=NamedRerouteUsage()
  )
  n25 = BumpOffset(
   Coordinate=NamedRerouteUsage()
   Height=LinearInterpolate(Alpha=@n24) {constA=-1}
   HeightRatioInput=ScalarParameter() {defaultValue=0.07, parameterName=BumpOffsetHeightRatio}
  )
  n26 = BumpOffset(
   Coordinate=NamedRerouteUsage()
   Height=LinearInterpolate() {constA=-1}
   HeightRatioInput=ScalarParameter() {defaultValue=0.07, parameterName=BumpOffsetHeightRatio}
  )
  n27 = NamedRerouteUsage()
  NamedRerouteDeclaration(None=LinearInterpolate(
    A=TextureSample(UVs=@n25) {samplerType=SAMPLERTYPE_Normal, texture=/Game/VisualThreatManager/StormVisuals/LargeClouds/T_StormCloudDetail01_N}.RGB
    B=TextureSample(UVs=@n26) {samplerType=SAMPLERTYPE_Normal, texture=&T_StormCloudDetail01_N}.RGB
    Alpha=@n27
   )) {name=Normal, nodeColor=<rgba 0.9219 0 1 1>}
  NamedRerouteDeclaration(None=LinearInterpolate(
    A=TextureSample(UVs=@n25) {samplerType=SAMPLERTYPE_LinearGrayscale, texture=&T_StormCloudDetail01_M}.G
    B=TextureSample(UVs=@n26) {samplerType=SAMPLERTYPE_LinearGrayscale, texture=&T_StormCloudDetail01_M}.G
    Alpha=@n27
   )) {name=GlowMask, nodeColor=<rgba 0 1 0.6016 1>}
  NamedRerouteDeclaration(None=@n24) {name=PreBump, nodeColor=<rgba 1 0 0.1875 1>}
  n28 = MaterialFunctionCall(Float2=TextureCoordinate()) {materialFunction=&BreakOutFloat2Components}
  n29 = AppendVector(
   A=Multiply(A=@n28.R, B=ScalarParameter() {defaultValue=1, parameterName=BottomRamp_Distortion_UTile})
   B=Multiply(A=@n28.G, B=ScalarParameter() {defaultValue=2, parameterName=BottomRamp_Distortion_VTile})
  )
  n30 = NamedRerouteUsage()
  NamedRerouteDeclaration(None=TextureSample(UVs=Panner(Coordinate=@n29, Time=@n30) {speedX=0.4}) {texture=&T_Cloud2dNoise}.R) {name=EmissiveNoise, nodeColor=<rgba 0 1 0.5313 1>}
  n31 = ScalarParameter() {defaultValue=1, parameterName=PanSpeed}
  NamedRerouteDeclaration(None=Multiply(A=Time(), B=Multiply(A=@n31, B=@n21.PanSpeed))) {name=time speed 2, nodeColor=<rgba 1 0 0.1172 1>}
  NamedRerouteDeclaration(None=Panner(Coordinate=NamedRerouteUsage(), Time=NamedRerouteUsage()) {speedX=1}) {name=pan 2, nodeColor=<rgba 0 1 0.2969 1>}
 ]
}
