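# Translucent, unlit, two-sided Niagara mesh particle material for the ring smoke of the CoreStreet skydiving contrail.
# A swirl texture, panned by a dynamic-parameter time, is tinted by particle color and shapes opacity with a vertex-red sine fade and view-angle fade.
# One slowly panning tiling noise modulates both opacity and SplineThicken world position offset.
# Dynamic parameters also drive thickness, noise offset and an optional texture-pan offset.
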Material /BRCosmetics/Athena/Effects/Particles/Skydiving/NS_Contrail_CoreStreet/M_Contrail_CoreStreet_RingSmoke
graph {
 properties {
  bAutomaticallySetUsageInEditor=false
  bUsedWithNiagaraMeshParticles=true
  blendMode=BLEND_Translucent
  shadingModel=MSM_Unlit
  twoSided=true
  usedShadingModels=Unlit
 }
 nodes [
  MaterialFunctionCall(Emissive=NamedRerouteUsage()) {materialFunction=&MF_FX_EmissiveAdjustments}
  n1 = VectorParameter() {parameterName=Texture_SV, defaultValue=<rgba 3 1 1 0>}
  n2 = LinearInterpolate(
   A=@n1.B
   B=Add(A=DynamicParameter() {parameterIndex=2, defaultValue=<rgba 1 1 1 0>, paramNames [AlphaOffset, "", "", ""]}.AlphaOffset, B=@n1.B)
   Alpha=ScalarParameter() {parameterName=UseAlphaOffset}
  )
  n3 = Add(A=MaterialFunctionCall!(
    InUV
    Texture
    InRGB=MaterialFunctionCall(X=@n1.R, Y=@n1.R, Z=@n2) {materialFunction=&MakeFloat3}.Result
    InA=@n1.A
    InTime=DynamicParameter() {paramNames [Time,Param2,Param3,Param4]}.Time
   ) {materialFunction=/Game/Effects/Fort_Effects/Materials/_Functions/MF_TexturePanner}.UV)
  n4 = NamedRerouteDeclaration(None=ChannelMaskParameter(None=TextureSampleParameter2D(UVs=@n3) {parameterName=MainTexture, texture=/BRCosmetics/Athena/Effects/Particles/Skydiving/NS_Contrail_SoilBlend/T_Trail_Swirl}.RGBA) {parameterName=TextureMask}.R) {name=Sampled Texture, nodeColor=<rgba 1 0 0.2344 1>}
  n5 = ScalarParameter() {defaultValue=1, parameterName=Color_Power}
  NamedRerouteDeclaration(None=Multiply(A=ParticleColor().RGB, B=Power(Base=@n4, Exp=@n5))) {name=OUT_Color, nodeColor=<rgba 1 0.75 0 1>}
  n6 = Constant()
  n7 = Multiply(A=TextureCoordinate(), B=AppendVector(A=Constant() {r=0.2}, B=@n6))
  n8 = AppendVector(A=DynamicParameter() {parameterIndex=2, defaultValue=<rgba 1 1 1 0>, paramNames [AlphaOffset, "", "", ""]}.AlphaOffset, B=@n6)
  n9 = TextureSample(UVs=Panner(Coordinate=Add(A=@n7, B=@n8)) {speedX=0.05}) {samplerType=SAMPLERTYPE_Grayscale, texture=/Game/Global/Textures/Default/Noise/TilingNoise06}
  n10 = ScalarParameter() {defaultValue=-0.1, parameterName=SplineThickenNoiseContrast}
  NamedRerouteDeclaration(None=LinearInterpolate(Alpha=MaterialFunctionCall(In=@n9.R, Contrast=@n10) {materialFunction=&CheapContrast}.Result)) {name=splinethicken_noise, nodeColor=<rgba 0 0.0859 1 1>}
  n11 = DynamicParameter() {parameterIndex=1, defaultValue=<rgba 1 1 1 0>, paramNames [SplineThicknessMult, "", "", U offset]}
  n12 = Multiply(A=ScalarParameter() {defaultValue=15, parameterName=SplineThickenBase}, B=@n11.SplineThicknessMult)
  n13 = NamedRerouteUsage()
  n14 = Multiply(A=ScalarParameter() {defaultValue=150, parameterName=SplineThickenTip}, B=@n11.SplineThicknessMult)
  n15 = NamedRerouteUsage()
  NamedRerouteDeclaration(None=MaterialFunctionCall(WidthBase=Multiply(A=@n12, B=@n13), WidthTip=Multiply(A=@n14, B=@n15), Expand U or V UV Channel=StaticBool()) {materialFunction=&SplineThicken}.WorldPositionOffset) {name=OUT_WPO, nodeColor=<rgba 1 0 0.2109 1>}
  n16 = NamedRerouteUsage()
  n17 = ScalarParameter() {defaultValue=1.2, parameterName=OpacityPower}
  n18 = ParticleColor()
  n19 = VertexColor()
  n20 = Power(Base=Sine(None=Multiply(A=@n19.R) {constB=0.5}))
  n21 = Multiply(A=Multiply(A=Power(Base=@n16, Exp=@n17), B=Multiply(A=@n18.A, B=@n20)), B=NamedRerouteUsage())
  n22 = ScalarParameter() {parameterName=Opacity Contrast}
  n23 = ScalarParameter() {defaultValue=1, parameterName=ViewAngleFade}
  NamedRerouteDeclaration(None=Multiply(
    A=MaterialFunctionCall(In=@n21, Contrast=@n22) {materialFunction=&CheapContrast}.Result
    B=MaterialFunctionCall("FadeSensitivity (optional)"=@n23) {materialFunction=&FadeBasedOnViewAngle}.Result
   )) {name=OUT_Opacity, nodeColor=<rgba 1 0.5859 0 1>}
  NamedRerouteDeclaration(None=MaterialFunctionCall(In=@n9.R, Contrast=ScalarParameter() {defaultValue=1, parameterName=OpacityNoiseContrast}) {materialFunction=&CheapContrast}.Result) {name=opacitynoise, nodeColor=<rgba 0.7109 0 1 1>}
  NamedRerouteUsage()
  NamedRerouteUsage()
 ]
}
