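# Translucent, unlit, two-sided Niagara sprite and mesh-particle material for the skulls in the CoreStreet skydiving contrail, using a decal texture.
# A time-driven sine, offset by a centered UV sphere mask, animates both color and a Z-axis displacement.
# Dynamic parameters blend the texture with a color-multiplied copy before an emissive-adjustment function, and scale the displacement.
# Texture alpha times particle alpha supplies opacity.
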Material /BRCosmetics/Athena/Effects/Particles/Skydiving/NS_Contrail_CoreStreet/M_Contrail_CoreStreet_Skulls
graph {
 properties {
  bUsedWithNiagaraMeshParticles=true
  bUsedWithNiagaraSprites=true
  blendMode=BLEND_Translucent
  pixelDepthOffsetMode=PDOM_Legacy
  shadingModel=MSM_Unlit
  twoSided=true
  usedShadingModels=Unlit
 }
 nodes [
  n1 = TextureSampleParameter2D() {parameterName=Texture, texture=&/T_FX_Contrail_CoreStreet_Decal}
  n2 = DynamicParameter() {paramNames [AmplitudeStrenght, ColorMult, ColorNoise, WPOStrenght]}
  n3 = Constant() {r=0.5}
  n4 = SphereMask(A=TextureCoordinate(), B=@n3, Radius=@n3, Hardness=Constant())
  n5 = Time()
  n6 = ScalarParameter() {defaultValue=2, parameterName=TimeMult}
  n7 = Sine(None=Add(A=Multiply(A=@n4) {constB=1.5}, B=Multiply(A=@n5, B=@n6) {constB=2}))
  MaterialFunctionCall(Emissive=LinearInterpolate(A=Multiply(A=@n1.RGB, B=@n2.ColorMult) {constB=0.8}, B=@n1.RGB, Alpha=LinearInterpolate(B=@n7, Alpha=@n2.ColorNoise) {constA=1})) {materialFunction=&MF_FX_EmissiveAdjustments}
  n8 = ScalarParameter() {defaultValue=8, parameterName=WPOStrenght}
  n9 = DynamicParameter() {paramNames [AmplitudeStrenght, ColorMult, ColorNoise, WPOStrenght]}
  Multiply(
   A=Multiply(A=Constant3Vector() {constant=<rgba 0 0 1 0>}, B=LinearInterpolate(A=@n8, Alpha=@n9.WPOStrenght) {constA=5, constB=0}) {constB=5}
   B=@n7
  )
  Multiply(A=ParticleColor().A, B=@n1.A)
 ]
}
